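; Gemstone-found quest actor for Battle Pass S30 Story 03, with a per-instance gem appearance on its pre-interaction mesh.
; The construction script creates a dynamic material from slot 0 of SM_Quest_PreInteraction, stores it as GemMI and reapplies it.
; It sets FlipbookAnimPhase from GemFlipbookValue, and sends StaticFlipbookAnimPhase as true through an audio-parameter interface message rather than a material setter.
; The BeginPlay, overlap and Tick events add no active logic.
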(event EventBeginPlay
  (|CallParentFunction) ; This node is disabled and will not be called.\nDrag off pins to build functionality.
)

(event Collision|EventActorBeginOverlap
  (params (OtherActor :type Actor%20Object%20Reference)))

(event EventTick
  (params (DeltaSeconds :type Float%20%28single-precision%29)))

(fn ConstructionScript
  (|CallParentFunction)
  (bind _returnvalue (Rendering|Material|CreateDynamicMaterialInstance (Rendering|Material|GetMaterial (Variables|Components|GetSM_Quest_PreInteraction) 0) "None" "None" :!class /Script/Engine.KismetMaterialLibrary))
  (Variables|Default|SetGemMI _returnvalue)
  (Rendering|Material|SetMaterial (Variables|Components|GetSM_Quest_PreInteraction) 0 (Variables|Default|GetGemMI))
  (Audio|Parameter|SetBooleanParameter(Message) (Variables|Default|GetGemMI) "StaticFlipbookAnimPhase" true)
  (Rendering|Material|SetScalarParameterValue (Variables|Default|GetGemMI) "FlipbookAnimPhase" (Variables|Default|GetGemFlipbookValue) :!class /Script/Engine.MaterialInstanceDynamic))
